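# Translucent unlit smoke material for Niagara mesh particles that turns mesh cards into camera-facing, looping smoke puffs with a tinted emissive.
# Vertex color randomizes each puff's rotation, texture offset into the inky smoke flipbook, size, growth speed and phase.
# A time-wrapped lifetime value drives sprite scale through World Position Offset.
# Opacity combines the texture, a radial mask, two SphereMasks that appear to fade over that lifetime, and DepthFade.
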Material /Game/Environments/Hera/FX/Materials/Generic/M_MeshSmokeSprite
graph {
 properties {
  bAutomaticallySetUsageInEditor=false
  bUsedWithNiagaraMeshParticles=true
  blendMode=BLEND_Translucent
  shadingModel=MSM_Unlit
  twoSided=true
  usedShadingModels=Unlit
 }
 nodes [
  MaterialFunctionCall(Emissive=VectorParameter() {group=Emissive, parameterName=SmokeColor, defaultValue=<rgba 1 1 1 1>}.RGB) {materialFunction=&MF_FX_EmissiveAdjustments}
  n1 = Time()
  n2 = ScalarParameter() {defaultValue=1, group=UV, parameterName=UV Rotation Speed}
  n3 = VertexColor()
  n4 = Rotator(Coordinate=TextureCoordinate(), Time=Add(A=Multiply(A=@n1, B=@n2), B=Multiply(B=@n3.R) {constA=20}))
  n5 = Multiply(A=@n3.B) {constB=0.5}
  n6 = Multiply(A=@n3.G) {constB=0.5}
  n7 = TextureSample(UVs=Add(A=Divide(A=@n4), B=AppendVector(A=@n5, B=@n6))) {texture=/Game/Effects/Niagara/MasterMaterials/Smoke/Textures/NEW/T_InkySmoke_Flipbook_blur}
  n8 = TextureCoordinate()
  n9 = Constant() {r=0.5}
  n10 = Multiply(
   A=Multiply(A=Multiply(A=@n7.RGB) {constB=2}, B=SphereMask(A=@n8, B=@n9) {attenuationRadius=0.5, hardnessPercent=0.5})
   B=SphereMask(
    A=NamedRerouteUsage()
    B=ScalarParameter() {group=Opacity, parameterName=StartPoint}
    Radius=ScalarParameter() {defaultValue=1, group=Opacity, parameterName=Param_4}
    Hardness=ScalarParameter() {group=Opacity, parameterName=Param_5}
   )
  )
  n11 = SphereMask(
   A=OneMinus(None=NamedRerouteUsage())
   B=ScalarParameter() {group=Opacity, parameterName=StartPoint}
   Radius=ScalarParameter() {defaultValue=1, group=Opacity, parameterName=Param_4}
   Hardness=ScalarParameter() {group=Opacity, parameterName=Param_5}
  )
  NamedRerouteDeclaration(None=DepthFade(Opacity=Multiply(A=@n10, B=@n11), FadeDistance=ScalarParameter() {defaultValue=100, group=Opacity, parameterName=DepthFade})) {name=Opacity, nodeColor=<rgba 0 0.5547 1 1>}
  n12 = VertexColor()
  n13 = LinearInterpolate(
   A=ScalarParameter() {defaultValue=200, group=WPO, parameterName=Random Scale Min}
   B=ScalarParameter() {defaultValue=250, group=WPO, parameterName=Random Scale}
   Alpha=@n12.G
  )
  n14 = Time()
  n15 = Add(A=ScalarParameter() {defaultValue=0.05, group=WPO, parameterName=GrowthSpeed}, B=Multiply(A=@n12.R) {constB=0.1})
  n16 = Fmod(A=Add(A=@n12.B, B=Multiply(A=@n14, B=@n15)), B=Constant() {r=1})
  n17 = Add(A=@n16, B=ScalarParameter() {defaultValue=0.5, group=WPO, parameterName=StartSizeOffset})
  NamedRerouteDeclaration(None=MaterialFunctionCall!(Center Location=WorldPosition().XYZ, 0-1 UVs, XY scale=Multiply(A=@n13, B=@n17), OverridePivot, "Normals (Optional)") {materialFunction=&Sprite}.World position offset) {name=WPO, nodeColor=<rgba 0 1 0.7891 1>}
  NamedRerouteDeclaration(None=@n16) {name=Lifetime, nodeColor=<rgba 0 1 0.6484 1>}
  NamedRerouteUsage()
  NamedRerouteUsage()
 ]
}
