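# Generic masked, Default Lit material for Niagara mesh particles, by its name intended for basic props in the Hera environment FX.
# It multiplies the RGB of a BaseTexture parameter by the particle color RGB, so emitters can tint each mesh.
# A separate NormalTexture parameter lets material instances swap surface detail.
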Material /Game/Environments/Hera/FX/Materials/Generic/M_Particle_Mesh_BasicProp
graph {
 properties {bAutomaticallySetUsageInEditor=false, bUsedWithNiagaraMeshParticles=true, blendMode=BLEND_Masked, usedShadingModels=Default Lit}
 nodes [
  Multiply(A=TextureSampleParameter2D() {parameterName=BaseTexture}.RGB, B=ParticleColor().RGB)
  TextureSampleParameter2D() {parameterName=NormalTexture}
 ]
}
