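# M_Particle_Mesh_Floating is a generic Hera environment FX material for Niagara mesh particles.
# It is a two-sided, masked, Default Lit surface with automatic usage flagging off.
# Its graph multiplies the RGB of a Color Texture parameter by particle color RGB, so each particle can tint the texture.
# No floating motion logic appears in the graph, nor does the wiring to material outputs.
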Material /Game/Environments/Hera/FX/Materials/Generic/M_Particle_Mesh_Floating
graph {
 properties {
  bAutomaticallySetUsageInEditor=false
  bUsedWithNiagaraMeshParticles=true
  bUsedWithStaticMesh=false
  blendMode=BLEND_Masked
  twoSided=true
  usedShadingModels=Default Lit
 }
 nodes [Multiply(A=TextureSampleParameter2D() {parameterName=Color Texture}.RGB, B=ParticleColor().RGB)]
}
